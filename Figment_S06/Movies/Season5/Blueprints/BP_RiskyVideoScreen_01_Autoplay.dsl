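; Drive-in video screen (Risky) that autoplays a non-looping video on randomized, authority-driven timers.
; ServerPlay flushes dormancy, starts playback and re-arms the timer; off mobile, playback fades a dynamic screen material before opening the media source, then runs another timeline at the end.
; StartVideo and the video-playback setting drive Initialize/Cleanup, with distance-based volume on mobile.
; Calendar-active and overlap-begin branches are empty, while overlap end cleans up.
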(event EventBeginPlay
  (Networking|SwitchHasAuthority
    (:Authority
      (if (not (Variables|Default|GetDebugBypassCalendarEvent))
        (MCPContext|WatchCalendarEvent (Variables|Default|GetEventString)
          (:OnEventActive
            _)))
      (bind _output_get (Variables|Default|SetPreviousDelay (Math|Random|RandomFloatinRange (Variables|Default|GetMin%20Delay) (Variables|Default|GetMax%20Delay))))
      (CallFunction|StartPlayTimer _output_get))))

(event Custom|StartPlayTimer
  (params (Time :type Float%20%28double-precision%29))
  ;[1 1 1] Delay Timer
  (Utilities|Time|SetTimerbyEvent (EventDispatchers|CreateEvent :!event Custom|ServerPlay) Time))

(event OnComponentBeginOverlap(Sphere)
  (params (OverlappedComponent :type Primitive%20Component%20Object%20Reference) (OtherActor :type Actor%20Object%20Reference) (OtherComp :type Primitive%20Component%20Object%20Reference) (OtherBodyIndex :type Integer) (bFromSweep :type Boolean) (SweepResult :type Hit%20Result%20Structure%20%28by%20ref%29))
  ;[1 1 1] Enable and disable video based on distance Sphere trigger volume
  (if (Utilities|Equal(Object) (Game|GetPlayerPawn 0) OtherActor)
    _))

(event Custom|StartVideo
  (bind _returnvalue (Settings|GetFortGameUserSettings))
  (Utilities|FlowControl|Delay 1e+01)
  ;[1 1 1] BeginPlay - Check calendar event active, if yes - Initialize media player, bind options delegate
  (EventDispatchers|BindEventtoOnAllowVideoPlaybackChanged _returnvalue (delegate Custom|OnAllowPlaybackChanged))
  (bind _returnvalue_2 (Class|FortGameUserSettings|IsVideoPlaybackAllowed _returnvalue))
  (if _returnvalue_2
    (CallFunction|Initialize)))

(event EventEndPlay
  (params (EndPlayReason :type EEndPlayReason%20Enum :default "Destroyed"))
  ;[1 1 1] EndPlay - Tidy media player, unbind delegate
  (EventDispatchers|UnbindEventfromOnAllowVideoPlaybackChanged (Settings|GetFortGameUserSettings) (delegate Custom|OnAllowPlaybackChanged))
  (CallFunction|Cleanup))

(event OnComponentEndOverlap(Sphere)
  (params (OverlappedComponent :type Primitive%20Component%20Object%20Reference) (OtherActor :type Actor%20Object%20Reference) (OtherComp :type Primitive%20Component%20Object%20Reference) (OtherBodyIndex :type Integer))
  ;[1 1 1] Enable and disable video based on distance Sphere trigger volume
  (if (Utilities|Equal(Object) (Game|GetPlayerPawn 0) OtherActor)
    ;[1 1 1] EndPlay - Tidy media player, unbind delegate
    (EventDispatchers|UnbindEventfromOnAllowVideoPlaybackChanged (Settings|GetFortGameUserSettings) (delegate Custom|OnAllowPlaybackChanged))
    (CallFunction|Cleanup)))

(event Custom|OnAllowPlaybackChanged
  (params (bAllowed :type Boolean))
  ;[1 1 1] Options delegate - update our state to match if needed
  (if (Math|Boolean|NotEqual(Boolean) bAllowed (Utilities|IsValid (Variables|Default|GetMediaPlayer) :!class /Script/Engine.KismetSystemLibrary))
    (if bAllowed
      (CallFunction|Initialize)
      (else
        (CallFunction|Cleanup)))))

(event Custom|Initialize
  ;[1 1 1] General initialization
  (Networking|CanExecuteCosmeticEvents
    (:True
      (CallFunction|CreateAndStartMediaPlayer)
      ;[1 1 1] Only kick this off if on a mobile platform
      (if (Fortnite|IsMobilePlatform)
        (Utilities|Time|SetTimerbyEvent (EventDispatchers|CreateEvent :!event Custom|Evaluate%20Volume%20Attenuation) (Variables|Default|GetTickRate) true)
        (CallFunction|EvaluateVolumeAttenuation)))))

(event Custom|ServerPlay
  (bind _currentdelay (Variables|Default|GetCurrentDelay))
  ;[1 1 1] Play on Server and reset playback timer for loop delay
  (Networking|FlushNetDormancy)
  (Networking|SwitchHasAuthority
    (:Authority
      (CallFunction|CreateAndStartMediaPlayer)
      (Variables|Default|SetCurrentDelay (Math|Random|RandomFloatinRange (Math|Float|Max(Float) (- (Variables|Default|GetMax%20Delay) (Variables|Default|GetPreviousDelay))) (Variables|Default|GetMax%20Delay)))
      (CallFunction|StartPlayTimer _currentdelay)
      (Variables|Default|SetPreviousDelay _currentdelay))))

(event Custom|Create%20and%20Start%20Media%20Player
  (bind _mediaplayer (Variables|Default|GetMediaPlayer))
  ;[1 1 1] Fade Diffuse Brightness MID
  (if (not (Fortnite|IsMobilePlatform))
    (bind _returnvalue (Rendering|Material|CreateDynamicMaterialInstance (Variables|Default|GetActivePlayerMaterial) "None" "None" :!class /Script/Engine.KismetMaterialLibrary))
    (bind _output_get (Variables|Default|SetActivePlayerMID _returnvalue))
    (Rendering|Material|SetMaterial (Variables|Default|GetDriveIn_Screen_Fixed) 0 _output_get)
    (bind (_direction _screenbrightness) (Timeline|ScreenFade PlayFromStart)
      (:Update
        ;[1 1 1] Create Media Player
        (Rendering|Material|SetScalarParameterValue (Variables|Default|GetActivePlayerMID) "BaseColor" _screenbrightness :!class /Script/Engine.MaterialInstanceDynamic))
      (:Finished
        (if (not (Utilities|IsValid (Variables|Default|GetMediaPlayer) :!class /Script/Engine.KismetSystemLibrary))
          (bind _returnvalue_2 (Game|ConstructObjectfromClass "/Script/MediaAssets.MediaPlayer" self))
          (Variables|Default|SetMediaPlayer _returnvalue_2)
          (Media|MediaPlayer|SetLooping (Variables|Default|GetMediaPlayer) false))
        (Media|MediaSoundComponent|SetMediaPlayer (Variables|Default|GetMediaSound) (Variables|Default|GetMediaPlayer))
        (Media|MediaTexture|SetMediaPlayer (Variables|Default|GetMedia%20Texture) (Variables|Default|GetMediaPlayer) :!class /Script/MediaAssets.MediaTexture)
        (Media|MediaPlayer|AssignOnMediaOpened _mediaplayer (EventDispatchers|CreateEvent :!event Custom|On%20Media%20Opened))
        (Media|MediaPlayer|BindEventtoOnEndReached _mediaplayer (EventDispatchers|CreateEvent :!event Custom|OnMediaFinishedPlaying))
        (Media|MediaPlayer|OpenSource (Variables|Default|GetMediaPlayer) (Variables|Default|GetPlatformSource))))))

(event Custom|On%20Media%20Opened
  (params (OpenedUrl :type String))
  ;[1 1 1] Play Media after opening
  (Variables|Default|SetMediaDuration (Math|Timespan|GetTotalSeconds (Media|MediaPlayer|GetDuration (Variables|Default|GetMediaPlayer))))
  (Variables|Default|SetInitialSeek (Math|Float|%25%28Float%29 (GameState|GetServerWorldTimeSeconds (Game|GetGameState)) (Variables|Default|GetMediaDuration)))
  (Media|MediaPlayer|Play (Variables|Default|GetMediaPlayer) :!class /Script/MediaAssets.MediaPlayer)
  (Media|MediaPlayer|SetNativeVolume (Variables|Default|GetMediaPlayer) 0.0))

(event Custom|OnMediaFinishedPlaying
  (bind (_direction _screenbrightness) (Timeline|Timeline_0 PlayFromStart)
    (:Update
      (Rendering|Material|SetScalarParameterValue (Variables|Default|GetActivePlayerMID) "BaseColor" _screenbrightness :!class /Script/Engine.MaterialInstanceDynamic))))

(event Custom|Cleanup
  ;[1 1 1] Cleanup
  (Rendering|Material|SetMaterial (Variables|Default|GetDriveIn_Screen_Fixed) 0 (Variables|Default|GetInactivePlayerMaterial))
  (if (Utilities|IsValid (Variables|Default|GetMediaPlayer) :!class /Script/Engine.KismetSystemLibrary)
    (Media|MediaPlayer|Close (Variables|Default|GetMediaPlayer) :!class /Script/MediaAssets.MediaPlayer)
    (Variables|Default|SetMediaPlayer 0)
    (Media|MediaSoundComponent|SetMediaPlayer (Variables|Default|GetMediaSound) 0)
    (Media|MediaTexture|SetMediaPlayer (Variables|Default|GetMedia%20Texture) 0 :!class /Script/MediaAssets.MediaTexture)))

(event Custom|Evaluate%20Volume%20Attenuation
  ;[1 1 1] Mobile platform volume attenuation
  (bind _returnvalue (Audio|AreAnyListenersWithinRange (Transformation|GetActorLocation) (Variables|Default|GetMaxAudibleDistance)))
  (if _returnvalue
    (Variables|Default|GetLocalFortPC
      (:then
        _)
      (:else
        (CallFunction|CacheLocalFortPC)
        (bind _localfortpc (Variables|Default|GetLocalFortPC)
          (:then
            _)
          (:else))))
    (Utilities|IsValid (Pawn|GetViewTarget _localfortpc)
      (:Is%20Valid
        (bind _distance (Class|MissionBlueprintFunctionLibrary|DistanceBetweenTwoVectors (Transformation|GetActorLocation (Pawn|GetViewTarget _localfortpc)) (Transformation|GetActorLocation)))
        (bind _output_get (Variables|Default|SetMobileVolumeTarget (Math|Float|MapRangeClamped _distance (Variables|Default|GetMinAudibleDistance) (Variables|Default|GetMaxAudibleDistance) 1.0)))
        (bind _returnvalue_2 (Media|MediaPlayer|SetNativeVolume (Variables|Default|GetMediaPlayer) _output_get))
        (Development|PrintString (Utilities|String|Append "Set Native Volume (" (Utilities|String|ToString(Boolean) _returnvalue_2) "): " (Utilities|String|ToString(Float) _output_get)) (Variables|Default|GetbDebugVolumeAttenuation) (Variables|Default|GetbDebugVolumeAttenuation))))
    (else
      (Media|MediaPlayer|SetNativeVolume (Variables|Default|GetMediaPlayer) 0.0))))

(fn CacheLocalFortPC
  (bind _returnvalue (Fortnite|GetLocalFortPlayerControllers true))
  (if (> (Utilities|Array|Length _returnvalue) 0)
    (Variables|Default|SetLocalFortPC (Utilities|Array|Get(acopy) _returnvalue))))

(fn ConstructionScript)
